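# Unlit additive, two-sided particle material for the LimeRock TV Charge Up effect, usable on Niagara sprites, ribbons and particle sprites.
# Particle color feeds MF_FX_EmissiveAdjustments, dimmed on back faces by a lerp on the clamped TwoSidedSign.
# A UV-centered SphereMask with Radius and Hardness parameters multiplies particle alpha, forming a procedural round dot without textures or depth fade.
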Material /LimeRock/Assets/FX/TV_Charge_Up/M_LimeRock_SoftParticle
graph {
 properties {
  bEnableResponsiveAA=true
  bUsedWithNiagaraRibbons=true
  bUsedWithNiagaraSprites=true
  bUsedWithParticleSprites=true
  blendMode=BLEND_Additive
  opacityMaskClipValue=0.03
  pixelDepthOffsetMode=PDOM_Legacy
  shadingModel=MSM_Unlit
  twoSided=true
  usedShadingModels=Unlit
 }
 nodes [
  n1 = ParticleColor()
  n2 = TwoSidedSign()
  MaterialFunctionCall(Emissive=LinearInterpolate(A=Multiply(A=@n1.RGB) {constB=0.15}, B=@n1.RGB, Alpha=Clamp(None=@n2))) {materialFunction=&MF_FX_EmissiveAdjustments}
  Multiply(
   A=@n1.A
   B=SphereMask(
    A=TextureCoordinate()
    B=Constant() {r=0.5}
    Radius=ScalarParameter() {defaultValue=0.5, parameterName=Radius}
    Hardness=ScalarParameter() {parameterName=Hardness}
   )
  )
 ]
}
